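# M_SteamMug_WaterSpouts is a material for the water spouts of the SteamMug Fountain A prop.
# It uses Default Lit shading and holds only a constant color vector and a scalar constant, with no textures, parameters or animation.
# Automatic setting of usage flags in the editor is turned off.
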Material /Hera/Props/SteamMug/SteamMug_Fountain_A/Materials/M_SteamMug_WaterSpouts
graph {
 properties {bAutomaticallySetUsageInEditor=false, usedShadingModels=Default Lit}
 nodes [
  Constant3Vector()
  Constant()
 ]
}
